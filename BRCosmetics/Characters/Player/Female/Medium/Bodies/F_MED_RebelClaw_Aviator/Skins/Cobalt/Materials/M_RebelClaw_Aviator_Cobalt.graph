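# Character material for the Cobalt skin of the female RebelClaw Aviator body, built on material attributes with the shading model set by expression.
# MF_CharacterParent_2023_PreFX supplies the base result, which passes through two optional static-switch effects borrowed from weapon skins: PanningEmissive, then PatternEmissiveFresnelSequence.
# Each effect takes its alpha from a channel-mask parameter on the Tech Art Mask.
# MF_CharacterParent_2023_PostFX finishes the chain.
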Material /BRCosmetics/Characters/Player/Female/Medium/Bodies/F_MED_RebelClaw_Aviator/Skins/Cobalt/Materials/M_RebelClaw_Aviator_Cobalt
graph {
 properties {
  bUseMaterialAttributes=true
  bUsedWithMeshDeformer=true
  bUsedWithMorphTargets=true
  bUsedWithSkeletalMesh=true
  pixelDepthOffsetMode=PDOM_Legacy
  shadingModel=MSM_FromMaterialExpression
  usedShadingModels=Default Lit | Subsurface | Clear Coat | Subsurface Profile | Cloth
 }
 nodes [
  n1 = MaterialFunctionCall() {materialFunction=&MF_CharacterParent_2023_PreFX}
  n2 = ChannelMaskParameter(None=@n1.Tech Art Mask) {group=05_Tech Art Masks, parameterName=Pattern Emissive Mask}
  n3 = StaticSwitchParameter(
   True=MaterialFunctionCall!(
    Alpha=ChannelMaskParameter(None=@n1.Tech Art Mask) {group=05_Tech Art Masks, parameterName=Panning Emissive Mask}.R
    isVehicle
    Gmap=ComponentMask(None=@n1.M) {g=true}
    UVOffset
    Local Normal
    Local Position=@n1.Pre skinned Local Position
    EmissiveColorOverride
    Material=@n1.Pre skinned Local Position
   ) {materialFunction=/Game/Weapons/WeaponSkins/MaterialFunction/PanningEmissive}.Result
   False=@n1.Result
  ) {group=04_Tech Art, parameterName=UsePanningEmissive, sortPriority=0}
  MaterialFunctionCall(
   Material=StaticSwitchParameter(
    True=MaterialFunctionCall!(Alpha=@n2.R, PanningEmissiveBrightnessAdjustment, Material=@n3) {materialFunction=/Game/Weapons/WeaponSkins/MaterialFunction/PatternEmissiveFresnelSequence}.Result
    False=@n3
   ) {group=04_Tech Art, parameterName=UsePatternEmissive, sortPriority=0}
   M=@n1.M
   Pre Skinned Local Position=@n1.M
   Distance Blend=@n1.M
   Fresnel=@n1.M
  ) {materialFunction=&MF_CharacterParent_2023_PostFX}
 ]
}
